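; Low-gravity directional launcher Blueprint that extends a parent launcher to fling players, wildlife and vehicles along its LaunchDirection.
; Launch events briefly ignore collision for player pawns, and for vehicles until overlap ends.
; ModifyCurrentLaunchVelocity boosts wildlife launches to compensate for their physics, otherwise restoring the original velocity or a tag-gated override.
; It also sends a quest stat for player launches and drives ambient audio plus direction-aligned idle and client-only burst VFX.
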(event EventBeginPlay
  ;[0 1 0.005187] BeginPlay
  ; VFX setup of the launch direction and call parent for enabled checks and storing original set launch velocity for possible overrides
  (|CallParentFunction)
  (Audio|Components|Audio|FadeIn (Variables|Default|GetAmbientAudio) 0.25 1.0 0.0 "Linear" :!class /Script/Engine.AudioComponent)
  ;;[0.5966 0.1136 1] Set direction of idle VFX on the launcher
  (Effects|Components|ParticleSystem|SetVectorParameter (Variables|Default|GetLoopIdle) "Direction" (Math|Rotator|GetRotationXVector (Transformation|GetWorldRotation (Variables|Default|GetLaunchDirection)))))

(event EventSoftDeactivation
  ;[1 1 1] If Soft Deactivated turn off FX and Audio
  (|CallParentFunction)
  (Components|Activation|Deactivate (Variables|Default|GetLoopIdle))
  (Audio|Components|Audio|FadeOut (Variables|Default|GetAmbientAudio) 1.0 0.0 "Linear" :!class /Script/Engine.AudioComponent))

(event EventPlayerLaunched
  (params (Player :type Fort%20Pawn%20Object%20Reference))
  (bind _returnvalue (Fortnite|GetFortPlayerControllerfromActor Player))
  ;[0 0.08696 1] Event (Actor) Launched
  ; Send quest event for tracking and call cosmetic events for VFX/SFX
  (bind _asfort_player_pawn (Utilities|Casting|CastToFortPlayerPawn Player)
    (:then
      (Collision|AddIgnoredActor _asfort_player_pawn self 0.5) ; don't collide with the launcher so that\nany attempted launch is not blocked
    )
    (:CastFailed
      (Utilities|Casting|CastToNPC_Pawn_Wildlife_Parent Player
        (:then
          (CallFunction|ModifyCurrentLaunchVelocity false (ScalableFloat|GetValueatLevel (Variables|Default|GetWildlifeVelocityMultiplier))) ; multiply wildlife launch velocity to land roughly \nnear the player land location \nPhysics settings (mass, grav scale, etc.) cause \nthem to not travel as far (FORT-511752)
        )
        (:CastFailed
          (CallFunction|ModifyCurrentLaunchVelocity true 1.0) ; reset velocity in case wildlife launched
        ))
      (|CallParentFunction Player)
      (Utilities|IsValid _returnvalue
        (:Is%20Valid
          (Quest|SendComplexCustomStatEvent (Quests|GetQuestManager _returnvalue "Athena") 0 "(GameplayTags=)" (Variables|Default|GetPlayerLaunchedTag) 1))
        (:Is%20Not%20Valid
          _))
      ;;[0.5966 0.1136 1] start Launched cosmetics VFX/SFX
      (CallFunction|ClientStartLaunchedCosmetics))))

(event EventVehicleLaunched
  (params (Vehicle :type Fort%20Athena%20Vehicle%20Object%20Reference))
  ;[0 0.08696 1] Event (Actor) Launched
  ; Send quest event for tracking and call cosmetic events for VFX/SFX
  (Collision|IgnoreActorwhenMoving (Variables|Default|GetMeshTemp) Vehicle true)
  (CallFunction|ModifyCurrentLaunchVelocity true 1.0) ; vehicles already have a modifier handled in the parent\nOnly reset the velocity in case wildlife launched
  (|CallParentFunction Vehicle)
  ;;[0.5966 0.1136 1] start Launched cosmetics VFX/SFX
  (CallFunction|ClientStartLaunchedCosmetics))

(event Collision|EventActorEndOverlap
  (params (OtherActor :type Actor%20Object%20Reference))
  (bind (_valid _asfort_athena_vehicle) (CallFunction|PlayerinVehicleorJustVehicle OtherActor))
  ;[0.3939 0.7539 1] Event ActorEndOverlap
  ; Re-enable collision for vehicles
  (|CallParentFunction OtherActor)
  (if _valid
    (Collision|IgnoreActorwhenMoving (Variables|Default|GetMeshTemp) _asfort_athena_vehicle false)))

(event Custom|Client_StartLaunchedCosmetics
  ;[0.5966 0.1136 1] Client_StartLaunchedCosmetics
  ; Setup and play Client Only VFX/SFX: on launch burst vfx
  (Networking|CanExecuteCosmeticEvents
    (:True
      (bind _returnvalue (Effects|SpawnFXSystemAttached :FXSystemAsset (Variables|FX|Visuals|GetVFX_LaunchBurst) :AttachToComponent (Variables|Default|GetMeshTemp) :AttachPointName "None" :Scale "1.000000,1.000000,1.000000" :LocationType "SnapToTarget" :bAutoDestroy true :PoolingMethod "None" :bAutoActivate true))
      (Effects|Components|ParticleSystem|SetVectorParameter _returnvalue "Direction" (Math|Rotator|GetRotationXVector (Transformation|GetWorldRotation (Variables|Default|GetLaunchDirection)))))))

(event EventEndPlay
  (params (EndPlayReason :type EEndPlayReason%20Enum :default "Destroyed"))
  ;[1 0.374 0] EventEndPlay
  ; Cleaup - clear timers, unbind events, etc.
  (|CallParentFunction EndPlayReason)
  (Utilities|Time|ClearandInvalidateTimerbyHandle (Variables|Default|GetReenableLaunchedActorCollisionTimer)))

(fn GetBaseVelocityLevel
  (return 0.0))

(fn ModifyCurrentLaunchVelocity
  (params (OnlyResetLaunchVelocity :type Boolean) (VelocityMultiplier :type Float%20%28double-precision%29))
  (if (not OnlyResetLaunchVelocity)
    (Utilities|FlowControl|Sequence
      (:then_0
        _)
      (:then_1
        (Variables|LaunchSettings|Gameplay|SetLaunchVelocity (Utilities|Struct|MakeScalableFloat (* VelocityMultiplier (ScalableFloat|GetValueatLevel (Variables|LaunchSettings|Gameplay|GetLaunchVelocity) (CallFunction|GetBaseVelocityLevel))))))))
  ;[0.1768 0.4743 1] Reset velocity & handle a velocity override case - Need to define desired Override Velocity and the Tag to check
  (if (GameplayTags|HasMatchingGameplayTag (Variables|LaunchSettings|Gameplay|GetTagToCheckForVelocityOverride))
    (Variables|LaunchSettings|Gameplay|SetLaunchVelocity (Variables|LaunchSettings|Gameplay|GetVelocityOverride))
    (else
      (Variables|LaunchSettings|Gameplay|SetLaunchVelocity (Variables|LaunchSettings|Gameplay|GetOriginalLaunchVelocity)))))

(fn ConstructionScript
  (bind (_returnvalue_roll _returnvalue_pitch _returnvalue_yaw) (Math|Rotator|BreakRotator (Transformation|GetWorldRotation (Variables|Default|GetMeshTemp))))
  (if (Variables|Default|GetFlattenRotationOnMesh)
    (Transformation|SetWorldRotation (Variables|Default|GetMeshTemp) (Math|Rotator|MakeRotator 0.0 _returnvalue_pitch _returnvalue_yaw) false false)
    (else
      _))
  (|CallParentFunction)
  (Effects|Components|ParticleSystem|SetVectorParameter (Variables|Default|GetLoopIdle) "Direction" (Math|Rotator|GetRotationXVector (Transformation|GetWorldRotation (Variables|Default|GetLaunchDirection)))))
