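# Unlit, two-sided translucent Niagara sprite and mesh material for the low-gravity launcher effect, drawing panning fluid-flame energy.
# Panning blurred noise distorts flame UVs, sampled thrice at slight offsets into RGB for chromatic fringing, then sharpened, saturation-boosted and tinted by particle color for emissive.
# Opacity combines desaturated flame, V gradient, distorted smoke-trail mask and particle alpha, with depth fade on default and deferred paths.
# IsParticle? switches speed and tiling between dynamic and scalar parameters.
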Material /DirectionalLauncherGameplay/Items/FX/M_LowGravLauncher
graph {
 properties {
  bUseTranslucencyVertexFog=false
  bUsedWithNiagaraMeshParticles=true
  bUsedWithNiagaraSprites=true
  blendMode=BLEND_Translucent
  pixelDepthOffsetMode=PDOM_Legacy
  shadingModel=MSM_Unlit
  twoSided=true
  usedShadingModels=Unlit
 }
 nodes [
  n1 = Time()
  Frac(None=@n1)
  n2 = Time()
  Frac(None=@n2)
  n3 = TextureCoordinate() {uTiling=2}
  n4 = DynamicParameter() {defaultValue=<rgba 0 0.2 1 1>, paramNames [SpeedX_01,SpeedY_01,SpeedX_02,SpeedY_02]}
  n5 = StaticSwitchParameter(
   True=AppendVector(A=@n4.SpeedX_01, B=@n4.SpeedX_01)
   False=AppendVector(A=ScalarParameter() {parameterName=SpeedX_01}, B=ScalarParameter() {defaultValue=0.2, parameterName=SpeedY_01})
  ) {parameterName=IsParticle?}
  n6 = DynamicParameter() {parameterIndex=1, defaultValue=<rgba 1 0.5 0.025 0.05>, paramNames [XTiler_Main,YTiler_Main,XTiler,YTiler]}
  n7 = ScalarParameter() {defaultValue=0.025, parameterName=XTiler}
  n8 = ScalarParameter() {defaultValue=0.05, parameterName=YTiler}
  n9 = Multiply(
   A=TextureSample!(UVs=Panner(Coordinate=@n3, Time=@n1, Speed=@n5) {speedY=-0.2}, Tex, Level, Apply View MipBias) {mipValueMode=TMVM_MipLevel, samplerType=SAMPLERTYPE_Grayscale, texture=&LowResBlurredNoise}.G
   B=StaticSwitchParameter(True=AppendVector(A=@n6.XTiler, B=@n6.XTiler), False=AppendVector(A=@n7, B=@n8)) {parameterName=IsParticle?}
  )
  n10 = MaterialFunctionCall(Float2=TextureCoordinate()) {materialFunction=&BreakOutFloat2Components}
  n11 = AppendVector(A=@n10.R, B=@n10.R)
  n12 = DynamicParameter() {parameterIndex=1, defaultValue=<rgba 1 0.5 0.025 0.05>, paramNames [XTiler_Main,YTiler_Main,XTiler,YTiler]}
  n13 = StaticSwitchParameter(
   True=AppendVector(A=@n12.XTiler_Main, B=@n12.XTiler_Main)
   False=AppendVector(A=ScalarParameter() {defaultValue=1, parameterName=XTiler_Main}, B=ScalarParameter() {defaultValue=0.5, parameterName=YTiler_Main})
  ) {parameterName=IsParticle?}
  n14 = DynamicParameter() {defaultValue=<rgba 0 0.2 1 1>, paramNames [SpeedX_01,SpeedY_01,SpeedX_02,SpeedY_02]}
  n15 = AppendVector(A=@n14.SpeedX_02, B=@n14.SpeedX_02)
  n16 = AppendVector(A=ScalarParameter() {parameterName=SpeedX_02}, B=ScalarParameter() {defaultValue=0.2, parameterName=SpeedY_02})
  n17 = Add(
   A=@n9
   B=Panner(Coordinate=Multiply(A=@n11, B=@n13), Time=@n2, Speed=StaticSwitchParameter(True=@n15, False=@n16) {parameterName=IsParticle?}) {speedY=0.2}
  )
  n18 = Constant() {r=0.002}
  n19 = TextureSampleParameter2D(UVs=Add(A=@n17, B=@n18)) {
   parameterName=Texture_01
   samplerType=SAMPLERTYPE_LinearGrayscale
   texture=/Game/Weapons/FORT_Melee/Pickaxe_Cube_Ninja_Male/FX/T_FluidFlameTile
  }
  n20 = TextureSampleParameter2D(UVs=@n17) {
   parameterName=Texture_02
   samplerType=SAMPLERTYPE_LinearGrayscale
   texture=&T_FluidFlameTile
  }
  n21 = TextureSampleParameter2D(UVs=Subtract(A=@n17, B=@n18)) {
   parameterName=Texture_03
   samplerType=SAMPLERTYPE_LinearGrayscale
   texture=&T_FluidFlameTile
  }
  n22 = MaterialFunctionCall(X=@n19.G, Y=@n20.G, Z=@n21.G) {materialFunction=&MakeFloat3}
  n23 = VectorParameter() {parameterName=Refine, defaultValue=<rgba 1 1 1 1>}
  n24 = Power(Base=@n22.Result, Exp=@n23.R) {constExponent=6}
  n25 = Desaturation(None=Add(A=Multiply(A=@n24, B=@n23.G) {constB=100}, B=Multiply(A=@n22.Result, B=@n23.B) {constB=5}), Fraction=Constant() {r=-0.5})
  n26 = ParticleColor()
  MaterialFunctionCall(Emissive=Multiply(A=Multiply(B=@n25) {constA=22}, B=@n26.RGB)) {materialFunction=&MF_FX_EmissiveAdjustments}
  n27 = Desaturation(None=@n22.Result, Fraction=Constant() {r=1})
  n28 = TextureCoordinate()
  n29 = ComponentMask(None=@n28) {g=true}
  n30 = TextureSample!(UVs=Add(A=@n9, B=@n28), Tex, Apply View MipBias) {samplerType=SAMPLERTYPE_Grayscale, texture=/Game/Effects/Fort_Effects/Textures/Smoke/T_Smoke_Trail_Soft_NoClamp}
  n31 = Multiply(A=@n26.A, B=Multiply(A=Multiply(A=@n27, B=@n23.A) {constB=10}, B=Multiply(A=@n29, B=@n30.R)))
  n32 = DepthFade(Opacity=@n31) {fadeDistanceDefault=20}
  ShadingPathSwitch(Default=@n32, Deferred=@n32, Forward=@n31, Mobile=@n31)
 ]
}
