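# Niagara burst effect for a directional low-gravity launcher, firing a blue energy blast along a direction, with User.Direction exposed.
# Ray, ring and organic cone meshes align to System.Direction, with the organic cones and streak sprites accelerating along it and spawning slower as the loop ages.
# Short Hermite-spline ribbons drift along the same direction.
# Light-emitting wispy and paired side glows, plus curl-noise sparks at both sides, dress the launcher.
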NiagaraSystem /DirectionalLauncherGameplay/Items/FX/NS_Launch_LowGravDirect_Burst
dependencies {
 usedDataInterfaces [
  &NiagaraDataInterfaceArrayFloat3
  &NiagaraDataInterfaceCurve
  &NiagaraDataInterfaceColorCurve
  &NiagaraDataInterfaceVectorCurve
 ]
 usedDynamicInputs []
 usedModules [
  &self:SystemSpawnScript.SystemScriptSource.SystemScriptGraph.NiagaraNodeAssignment_0.SetVariables_#1
  &SystemState
  &self:SystemSpawnScript.SystemScriptSource.SystemScriptGraph.NiagaraNodeAssignment_1.SetVariables_#2
 ]
}
emitters {
 Cone_Organic {
  emitterUpdateScript [
   &EmitterState(Life Cycle Mode=::NewEnumerator1, Loop Behavior=::NewEnumerator1, Loop Delay=0.05, Loop Duration=0.55, UseLoopDelay=true)
   &SpawnRate(Spawn Probability=0.5, SpawnRate=33, Use Spawn Probability=false)
  ]
  particleSpawnScript [
   &InitializeParticle(
    Lifetime=0.3
    Mesh Scale=&MakeVector(X=1.5, Y=1.5, Z=&ClampFloat(Float=$Emitter.CurrentLoopDuration, Min=0.5, Max=2))
    Mesh Scale Mode=::NewEnumerator4
    Sprite Size Mode=::NewEnumerator3
    Uniform Sprite Size=50
    UsePositionOffset=false
    Write Material Random=true
   )
   &InitialMeshOrientation(
    Orientation Axis=<0 0 1>
    Orientation Vector=$System.Direction
    Rotation=&RandomRangeVector(Minimum=<0 0 0>, Maximum=<0 0 1>)
    Use Orientation Vector=true
    Use Rotation Vector=true
   ) {module=InitialMeshOrientation001}
   &SystemLocation(
    MaskSpawnGroup=false
    Offset=&Add_Vector(
     A=$System.OffsetLocation
     B=&Multiply_VectorByFloat(
      Vector=$System.Direction
      Float=&Multiply_Float(
       A=-88
       B=&MakeFloatFromInt(Int=&Add_Integer(A=1, B=&ReturnExecIndex()))
      )
     )
    )
    Offset Coordinate Space=::World
   )
   &AddVelocity(
    Rotation Coordinate Space=::World
    Velocity=$System.Direction
    Velocity Speed Scale=&Multiply_Float(
     A=3200
     B=&FloatFromCurve(
      FloatCurve=DataInterface() {propertyValues {Curve {keys [
          {interpMode=RCIM_Cubic, tangentMode=RCTM_User, time=0, value=1}
          {arriveTangent=-1.9179, interpMode=RCIM_Cubic, leaveTangent=-1.9179, tangentMode=RCTM_User, time=1, value=0.15}
         ]}}}
      CurveIndex=$Emitter.NormalizedLoopAge
     )
    )
   )
  ]
  particleUpdateScript [
   &ParticleState()
   &ScaleMeshSize(Scale Factor=&VectorFromFloat(Value=&FloatFromCurve(
      FloatCurve=DataInterface() {propertyValues {Curve {keys [
          {interpMode=RCIM_Cubic, tangentMode=RCTM_User, time=0, value=0.55}
          {arriveTangent=2.1267, interpMode=RCIM_Cubic, leaveTangent=2.1267, tangentMode=RCTM_User, time=1, value=1}
         ]}}}
      Scale Curve=2
     ))) {module=ScaleMeshSize001}
   &ScaleColor(
    Scale Alpha=&FloatFromCurve(
     FloatCurve=DataInterface() {propertyValues {Curve {keys [
         {interpMode=RCIM_Cubic, tangentMode=RCTM_User, time=0, value=0}
         {arriveTangent=5.5316, interpMode=RCIM_Cubic, tangentMode=RCTM_Break, time=0.5, value=1}
         {tangentMode=RCTM_User, time=1, value=0}
        ]}}}
     Scale Curve=0.2
    )
    Scale RGB=&MakeVectorFromLinearColorRGB(Color=&ColorFromCurve(ColorCurve=DataInterface() {propertyValues {
        AlphaCurve {keys [<at 0 1 RCIM_Cubic>,<at 1 0 RCIM_Cubic>]}
        BlueCurve {keys [<at 0.511 1 RCIM_Cubic>,<at 0.5507 1>,<at 0.5722 1>,<at 1 1>]}
        GreenCurve {keys [<at 0.511 0.4289 RCIM_Cubic>,<at 0.5507 0.318>,<at 0.5722 0.0612>,<at 1 0.1843>]}
        RedCurve {keys [<at 0.511 0 RCIM_Cubic>,<at 0.5507 0.0859>,<at 0.5722 0>,<at 1 0>]}
       }}))
    ScaleA=true
    ScaleRGB=true
   )
   &ScaleColor(
    Scale Alpha=&FloatFromCurve(
     FloatCurve=DataInterface() {propertyValues {Curve {keys [
         {interpMode=RCIM_Cubic, leaveTangentWeight=0.4633, tangentMode=RCTM_User, time=0.1, value=1}
         {arriveTangentWeight=0.3377, interpMode=RCIM_Cubic, tangentMode=RCTM_User, time=0.6, value=0}
        ]}}}
     CurveIndex=$Emitter.Age
    )
    Scale RGB=&VectorFromFloat(Value=&Lerp_Float(
      A=3
      B=1
      Alpha=&FloatFromCurve(
       FloatCurve=DataInterface() {propertyValues {Curve {keys [
           {interpMode=RCIM_Cubic, leaveTangentWeight=0.4633, tangentMode=RCTM_User, time=0, value=1}
           {arriveTangentWeight=0.8688, interpMode=RCIM_Cubic, tangentMode=RCTM_User, time=0.6, value=0}
          ]}}}
       CurveIndex=$Emitter.Age
      )
     ))
    ScaleA=true
    ScaleRGB=true
   ) {module=ScaleColor001}
   &AccelerationForce(Acceleration=&Multiply_VectorByFloat(Vector=$System.Direction, Float=2222)) {module=AccelerationForce002}
   &SolveForcesAndVelocity(Clamp Velocity=false, Limit Acceleration=false)
   &DynamicMaterialParameters(
    Index 0 Param 1=-0.1
    Index 0 Param 2=0.25
    Index 0 Param 3=0.1
    Index 0 Param 4=0.25
    Module.Param0WriteEnabled=true
    Module.Param1WriteEnabled=true
    Module.Param2WriteEnabled=true
    Module.Param3WriteEnabled=true
   ) {module=DynamicMaterialParameters001}
  ]
  rendererClasses [&NiagaraMeshRendererProperties]
  renderers [{rendererClass=&NiagaraMeshRendererProperties, rendererIndex=0}]
 }
 Cone_Ray {
  emitterUpdateScript [
   &EmitterState()
   &SpawnBurst_Instantaneous(Spawn Count=1, Use Spawn Probability=false)
  ]
  particleSpawnScript [
   &InitializeParticle(
    Color=<rgba 0 0.8333 5 0.07>
    Lifetime=0.7
    Mesh Scale=&Multiply_Vector(A=<1 0.8 1>, B=&VectorFromFloat(Value=5))
    Mesh Scale Mode=::NewEnumerator4
    UsePositionOffset=false
    Write Material Random=true
   ) {module=InitializeParticle001}
   &InitialMeshOrientation(
    Orientation Axis=<0 0 -1>
    Orientation Vector=$System.Direction
    Rotation=&RandomRangeVector(Minimum=<0 0 0>, Maximum=<0 0 1>)
    Use Orientation Vector=true
    Use Rotation Vector=false
   ) {module=InitialMeshOrientation002}
   &SystemLocation(MaskSpawnGroup=false, Offset=$System.OffsetLocation)
  ]
  particleUpdateScript [
   &ParticleState()
   &ScaleColor(
    Scale Alpha=&FloatFromCurve(FloatCurve=DataInterface() {propertyValues {Curve {keys [
         {interpMode=RCIM_Cubic, tangentMode=RCTM_Break, time=0, value=0}
         {arriveTangent=28.2007, interpMode=RCIM_Cubic, tangentMode=RCTM_Break, time=0.1, value=1}
         {arriveTangent=-0.0405, interpMode=RCIM_Cubic, leaveTangent=-12.7374, tangentMode=RCTM_Break, time=0.6, value=1}
         {arriveTangent=-0.5337, interpMode=RCIM_Cubic, leaveTangent=-0.5337, tangentMode=RCTM_User, time=1, value=0}
        ]}}})
    ScaleA=true
    ScaleRGB=true
   )
  ]
  rendererClasses [&NiagaraMeshRendererProperties]
  renderers [{rendererClass=&NiagaraMeshRendererProperties, rendererIndex=0}]
 }
 Cone_Rings {
  emitterUpdateScript [
   &EmitterState()
   &SpawnBurst_Instantaneous(Spawn Count=1, Use Spawn Probability=false)
  ]
  particleSpawnScript [
   &InitializeParticle(
    Color=<rgba 1 1 1 0.15>
    Lifetime=0.5
    Mesh Scale=<2 2 5>
    Mesh Scale Mode=::NewEnumerator4
    UsePositionOffset=false
    Write Material Random=true
   ) {module=InitializeParticle001}
   &InitialMeshOrientation(Orientation Axis=<0 0 -1>, Orientation Vector=$System.Direction, Use Orientation Vector=true, Use Rotation Vector=false) {module=InitialMeshOrientation001}
   &SystemLocation(MaskSpawnGroup=false, Offset=$System.OffsetLocation)
  ]
  particleUpdateScript [
   &ParticleState()
   &SolveForcesAndVelocity(Clamp Velocity=false, Limit Acceleration=false)
   &ScaleColor(
    Scale Alpha=&FloatFromCurve(FloatCurve=DataInterface() {propertyValues {Curve {keys [
         {arriveTangent=9.0725, interpMode=RCIM_Cubic, leaveTangent=9.0725, tangentMode=RCTM_User, time=0, value=0}
         {arriveTangent=15.0597, interpMode=RCIM_Cubic, leaveTangent=-2.5249, tangentMode=RCTM_Break, time=0.1, value=0.9847}
         {arriveTangent=-0.0405, interpMode=RCIM_Cubic, leaveTangent=-0.0405, tangentMode=RCTM_User, time=1, value=0}
        ]}}})
    Scale RGB=&MakeVectorFromLinearColorRGB(Color=<rgba 1.2082 0.9917 9 1>)
    ScaleA=true
    ScaleRGB=true
   ) {module=ScaleColor001}
  ]
  rendererClasses [&NiagaraMeshRendererProperties]
  renderers [{rendererClass=&NiagaraMeshRendererProperties, rendererIndex=0}]
 }
 Cone_Sprites {
  emitterUpdateScript [
   &EmitterState()
   &SpawnRate(SpawnRate=&RandomRangeFloat(Minimum=5, Maximum=8), Use Spawn Probability=false)
  ]
  particleSpawnScript [
   &InitializeParticle(
    Color=<rgba 0 4.6 9 1>
    Lifetime Max=0.5
    Lifetime Min=0.3
    Lifetime Mode=::NewEnumerator1
    Position Offset=$System.OffsetLocation
    Sprite Size Mode=::NewEnumerator1
    Uniform Sprite Size Max=4
    Uniform Sprite Size Min=1
    UsePositionOffset=true
    Write Material Random=true
   )
   &ShapeLocation(
    Cylinder Radius=40
    MaskSpawnGroup=false
    Override Seed=false
    Rotation Axis=$System.Direction
    Rotation Coordinate Space=::World
    Rotation Mode=::NewEnumerator1
    Shape Primitive=::NewEnumerator1
    Surface Only=false
    UseOwnerScale=true
   )
   &AddVelocity(
    Rotation Coordinate Space=::World
    Velocity=$System.Direction
    Velocity Speed Scale=&Multiply_Float(
     A=&RandomRangeFloat(Minimum=1200, Maximum=2000)
     B=&FloatFromCurve(
      FloatCurve=DataInterface() {propertyValues {Curve {keys [
          {interpMode=RCIM_Cubic, tangentMode=RCTM_User, time=0, value=1}
          {arriveTangent=-1.9179, interpMode=RCIM_Cubic, leaveTangent=-1.9179, tangentMode=RCTM_User, time=1, value=0.15}
         ]}}}
      CurveIndex=$Emitter.NormalizedLoopAge
     )
    )
   )
  ]
  particleUpdateScript [
   &ParticleState()
   &ScaleSpriteSize(Scale Factor=&Vector2DFromFloat(Value=&FloatFromCurve(FloatCurve=DataInterface() {propertyValues {Curve {keys [{arriveTangent=-5.2934, interpMode=RCIM_Cubic, leaveTangent=-5.2934, tangentMode=RCTM_User, time=0.7, value=1}, <at 1 0>]}}})))
   &AccelerationForce(Acceleration=&Multiply_VectorByFloat(Vector=$System.Direction, Float=1000))
   &SolveForcesAndVelocity(Clamp Velocity=false, Limit Acceleration=false)
   &ScaleSpriteSizeBySpeed(Max Scale Factor=<1 99>, Velocity Threshold=4250)
   &ScaleColor(
    Scale Alpha=&FloatFromCurve(FloatCurve=DataInterface() {propertyValues {Curve {keys [
         {tangentMode=RCTM_User, time=0, value=1}
         {arriveTangent=-0.0405, interpMode=RCIM_Cubic, leaveTangent=-0.0405, tangentMode=RCTM_User, time=0.7009, value=1}
         {arriveTangent=-8.0089, interpMode=RCIM_Cubic, leaveTangent=-8.0089, tangentMode=RCTM_User, time=1, value=0}
        ]}}})
    ScaleA=true
    ScaleRGB=false
   )
   &ScaleColor(
    Scale Alpha=&FloatFromCurve(
     FloatCurve=DataInterface() {propertyValues {Curve {keys [
         {interpMode=RCIM_Cubic, leaveTangentWeight=0.4633, tangentMode=RCTM_User, time=0.1, value=1}
         {arriveTangentWeight=0.3377, interpMode=RCIM_Cubic, tangentMode=RCTM_User, time=0.6, value=0}
        ]}}}
     CurveIndex=$Emitter.Age
    )
    Scale RGB=&VectorFromFloat(Value=&Lerp_Float(
      A=3
      B=1
      Alpha=&FloatFromCurve(
       FloatCurve=DataInterface() {propertyValues {Curve {keys [
           {interpMode=RCIM_Cubic, leaveTangentWeight=0.4633, tangentMode=RCTM_User, time=0, value=1}
           {arriveTangentWeight=0.8688, interpMode=RCIM_Cubic, tangentMode=RCTM_User, time=0.6, value=0}
          ]}}}
       CurveIndex=$Emitter.Age
      )
     ))
    ScaleA=true
    ScaleRGB=true
   ) {module=ScaleColor001}
  ]
  rendererClasses [&NiagaraSpriteRendererProperties]
  renderers [{rendererClass=&NiagaraSpriteRendererProperties}]
 }
 Glow_Idle {
  emitterUpdateScript [
   &EmitterState()
   &SpawnBurst_Instantaneous(Use Spawn Probability=false, UseLoopCountLimit=false)
   &SpawnBurst_Instantaneous(Spawn Group=1, Use Spawn Probability=false, UseLoopCountLimit=false) {module=SpawnBurst_Instantaneous001}
  ]
  particleSpawnScript [
   &FN_SetSpawnGroup(Number of Spawn Groups=2)
   &InitializeParticle(
    Color=<rgba 0.41 1.5469 3 0.5>
    Lifetime=1.5
    Position Offset=&SelectVectorFromArray(Direct Array Index=$Particles.SpawnGroup, Vector Selection Array=DataInterface() {propertyValues {FloatData [<200 0 130>, <-200 0 130>]}})
    Sprite Rotation Mode=::NewEnumerator0
    Sprite Size=<222 222>
    Sprite Size Mode=::NewEnumerator4
    UsePositionOffset=true
    Write Material Random=true
   ) {module=InitializeParticle001}
  ]
  particleUpdateScript [
   &ParticleState()
   &DynamicMaterialParameters(
    Index 0 Param 1=2.484
    Index 0 Param 2=3
    Module.Param0WriteEnabled=true
    Module.Param1WriteEnabled=true
    Module.Param2WriteEnabled=true
    Module.Param3WriteEnabled=true
   )
   &SolveForcesAndVelocity(Clamp Velocity=false, Limit Acceleration=false)
   &ScaleColor(
    Scale Alpha=&FloatFromCurve(
     FloatCurve=DataInterface() {propertyValues {Curve {keys [
         {tangentMode=RCTM_User, time=0, value=1}
         {arriveTangent=-0.0405, interpMode=RCIM_Cubic, leaveTangent=-0.0405, tangentMode=RCTM_User, time=0.8, value=1}
         {arriveTangent=0.6935, interpMode=RCIM_Cubic, leaveTangent=0.6935, tangentMode=RCTM_User, time=1, value=0}
        ]}}}
     Scale Curve=0.06
    )
    ScaleA=true
    ScaleRGB=false
   ) {module=ScaleColor001}
   &ScaleSpriteSize(Uniform Curve Sprite Scale=DataInterface() {propertyValues {Curve {keys [
        {interpMode=RCIM_Cubic, tangentMode=RCTM_User, time=0, value=0.8}
        {interpMode=RCIM_Cubic, tangentMode=RCTM_User, time=0.5, value=1}
        {tangentMode=RCTM_User, time=1, value=0.8}
       ]}}})
  ]
  rendererClasses [&NiagaraSpriteRendererProperties, &NiagaraLightRendererProperties]
  renderers [
   {rendererClass=&NiagaraSpriteRendererProperties}
   {rendererClass=&NiagaraLightRendererProperties, rendererIndex=1}
  ]
 }
 Glow_Wispy {
  emitterUpdateScript [
   &EmitterState()
   &SpawnBurst_Instantaneous(Use Spawn Probability=false, UseLoopCountLimit=false)
  ]
  particleSpawnScript [
   &InitializeParticle(
    Color=<rgba 1.2522 3.5652 4 1>
    Lifetime=0.7
    Sprite Rotation Angle=&RandomRangeFloat(Minimum=0, Maximum=359)
    Sprite Rotation Mode=::NewEnumerator0
    Sprite Size Mode=::NewEnumerator3
    Uniform Sprite Size=250
    UsePositionOffset=false
    Write Material Random=true
   ) {module=InitializeParticle001}
   &SystemLocation(MaskSpawnGroup=false, Offset=<0 0 33>, Offset Coordinate Space=::World)
   &CameraOffset(Camera Offset Amount=20)
   &DynamicMaterialParameters(
    Index 0 Param 1=-1.1
    Module.Param0WriteEnabled=true
    Module.Param1WriteEnabled=false
    Module.Param2WriteEnabled=false
    Module.Param3WriteEnabled=false
   )
  ]
  particleUpdateScript [
   &ParticleState()
   &ScaleSpriteSize(Scale Sprite Size Mode=::NewEnumerator2)
   &ScaleColor(
    Scale Alpha=&FloatFromCurve(FloatCurve=DataInterface() {propertyValues {Curve {keys [
         {arriveTangent=13.8526, interpMode=RCIM_Cubic, leaveTangent=13.8526, tangentMode=RCTM_User, time=0, value=0}
         <at 0.2 0.9>
         {interpMode=RCIM_Cubic, tangentMode=RCTM_User, time=1, value=0}
        ]}}})
    ScaleA=true
    ScaleRGB=false
   ) {module=ScaleColor001}
  ]
  rendererClasses [&NiagaraSpriteRendererProperties, &NiagaraLightRendererProperties]
  renderers [
   {rendererClass=&NiagaraSpriteRendererProperties}
   {rendererClass=&NiagaraLightRendererProperties, rendererIndex=1}
  ]
 }
 Ribbons {
  emitterUpdateScript [
   &EmitterState()
   &SpawnBurst_Instantaneous(Spawn Count=&RandomRangeInt(Minimum=4, Maximum=6), Use Spawn Probability=false)
  ]
  particleSpawnScript [
   &InitializeRibbon(
    Lifetime=&RandomRangeFloat(Minimum=0.3, Maximum=0.5, Randomness Mode=::NewEnumerator1, Random Seed=$Emitter.LoopCount, Fixed Random Seed=true)
    Module.Write Color=true
    Module.Write Lifetime=true
    Module.Write Mass=true
    Module.Write Material Random=true
    Module.Write Position=false
    Module.Write RibbonTwist=false
    Module.Write RibbonWidth=true
    Ribbon Width=20
   ) {module=InitializeRibbon001}
   /SetVariables(
    Particles.RibbonID=/Niagara/DynamicInputs/NiagaraID/MakeNiagaraIDFromInt(Index:NiagaraInt32=$Emitter.LoopCount, Acquire Tag:NiagaraInt32=0)
    Particles.RibbonLinkOrder:NiagaraFloat=&ReturnNormalizedExecIndex()
   ) {isSetParameters=true}
   /Game/Effects/Niagara/Modules/Curves/FN_HermiteCubicSpline_v1/FN_HermiteCubicSpline(
    Position Blend:NiagaraFloat=$Particles.Initial.RibbonLinkOrder
    Position End:Vector3f=&Add_Vector(
     A=&Multiply_Vector(
      A=&RandomVector(Vector Scale=&RandomRangeFloat(Minimum=10, Maximum=50))
      B=&RandomRangeVector(Minimum=<0 0 0>, Maximum=<0 1 1>)
     )
     B=<50 0 99>
    )
    Position Start:Vector3f=&Add_Vector(
     A=&Multiply_Vector(
      A=&RandomVector(Vector Scale=&RandomRangeFloat(Minimum=25, Maximum=40))
      B=&RandomRangeVector(Minimum=<1 1 1>, Maximum=<1 1 1>)
     )
     B=<0 0 0>
    )
    Write Particles Position=true
   )
   &AddVelocity(
    Coordinate Space=::World
    Velocity=&Multiply_Vector(
     A=&VectorFromFloat(Value=&Add_Float(
       A=&Multiply_Float(A=3, B=&ReturnNormalizedExecIndex())
       B=&ReturnNormalizedExecIndex()
      ))
     B=&Multiply_VectorByFloat(Vector=$System.Direction, Float=&RandomRangeFloat(Minimum=50, Maximum=150))
    )
   ) {module=AddVelocity001}
  ]
  particleUpdateScript [
   &ParticleState()
   &ScaleColor(
    Scale Alpha=&FloatFromCurve(
     FloatCurve=DataInterface() {propertyValues {Curve {keys [{arriveTangent=-5.9757, interpMode=RCIM_Cubic, time=0, value=0}, <at 0.5 0.999 RCIM_Cubic>, <at 1 0>]}}}
     Scale Curve=&FloatFromCurve(
      FloatCurve=DataInterface() {propertyValues {Curve {keys [{arriveTangent=-1.9917, interpMode=RCIM_Cubic, leaveTangent=-1.9917, tangentMode=RCTM_User, time=-0.0108, value=1}, <at 1 0 RCIM_Cubic>]}}}
      CurveIndex=&Multiply_Float(A=$Particles.RibbonLinkOrder, B=$Particles.NormalizedAge)
      Scale Curve=0.5
     )
    )
    Scale RGB=&MakeVectorFromLinearColorRGB(Color=<rgba 0 0.7292 3 1>)
    ScaleA=true
    ScaleRGB=true
   )
   &ScaleRibbonWidth(Ribbon Width Scale=&FloatFromCurve(FloatCurve=DataInterface() {propertyValues {Curve {keys [
         {arriveTangent=-6.3997, interpMode=RCIM_Cubic, time=0, value=1}
         {arriveTangent=-1, interpMode=RCIM_Cubic, leaveTangent=-1, time=0.5, value=1}
         {arriveTangent=-2.2297, interpMode=RCIM_Cubic, leaveTangent=-2.2297, tangentMode=RCTM_User, time=1, value=0}
        ]}}}))
   &SolveForcesAndVelocity(Clamp Velocity=false, Limit Acceleration=false)
   &DynamicMaterialParameters(
    Index 0 Param 1=1
    Index 0 Param 3=0.1
    Index 1 Param 4=200
    Module.Param0WriteEnabled=true
    Module.Param0WriteEnabled1=true
    Module.Param1WriteEnabled=true
    Module.Param1WriteEnabled1=true
    Module.Param2WriteEnabled=true
    Module.Param2WriteEnabled1=true
    Module.Param3WriteEnabled=true
    Module.Param3WriteEnabled1=true
    Write Parameter Index 1=true
   )
  ]
  rendererClasses [&NiagaraRibbonRendererProperties]
  renderers [{rendererClass=&NiagaraRibbonRendererProperties}]
 }
 SparksA {
  emitterUpdateScript [
   &EmitterState()
   &SpawnBurst_Instantaneous(
    Spawn Count=&RandomRangeInt(Minimum=3, Maximum=7)
    Spawn Probability=0.1
    Use Spawn Probability=true
    UseLoopCountLimit=false
   )
   &SpawnRate(SpawnRate=&RandomRangeFloat(Minimum=8, Maximum=20), Use Spawn Probability=false)
  ]
  particleSpawnScript [
   &InitializeParticle(
    Color=<rgba 0 3.8042 9 0.4>
    Lifetime Max=0.6
    Lifetime Min=0.4
    Lifetime Mode=::NewEnumerator1
    Position Offset=<-200 0 154>
    Sprite Rotation Mode=::NewEnumerator2
    Sprite Size Mode=::NewEnumerator1
    Sprite UV Mode=::NewEnumerator3
    Uniform Sprite Size Max=3.5
    Uniform Sprite Size Min=2
    UsePositionOffset=true
    Write Material Random=true
   ) {module=InitializeParticle001}
   &ShapeLocation(MaskSpawnGroup=false, Override Seed=false, Sphere Radius=50, UseOwnerScale=true) {module=ShapeLocation001}
   &AddVelocity(Velocity=<55 50 50>)
  ]
  particleUpdateScript [
   &ParticleState()
   &CurlNoiseForce(Noise Frequency=25, Noise Strength=5, Pan Noise=true, Pan Noise Field=<0 0 0.1>)
   &Drag(Drag=&RandomRangeFloat(Minimum=0.8, Maximum=1.2), Use Linear Drag=true, Use Rotational Drag=false)
   &ScaleSpriteSize(Scale Factor=&Vector2DFromFloat(Value=&FloatFromCurve(FloatCurve=DataInterface() {propertyValues {Curve {keys [
          <at 0 0 RCIM_Cubic>
          <at 0.112 1.0173>
          {arriveTangent=-1.142, interpMode=RCIM_Cubic, leaveTangent=-1.142, time=0.7001, value=0.872}
          <at 1.0011 0.002>
         ]}}})))
   &AccelerationForce(
    Acceleration=&VectorFromCurve(
     VectorCurve=DataInterface() {propertyValues {
       XCurve {keys [
         {interpMode=RCIM_Cubic, leaveTangent=3.2171, tangentMode=RCTM_Break, time=0, value=0}
         {interpMode=RCIM_Cubic, tangentMode=RCTM_Break, time=1, value=1}
        ]}
       YCurve {keys [<at 0 1>,<at 1 0>]}
       ZCurve {keys [<at 0 1>,<at 1 0>]}
      }}
     Scale Curve=<9999 2222 1>
    )
    Coordinate Space=::Local
   ) {module=AccelerationForce002}
   &SolveForcesAndVelocity(Clamp Velocity=false, Limit Acceleration=false)
   &ScaleSpriteSizeBySpeed(Max Scale Factor=<1 5>, Velocity Threshold=250)
   &ScaleSpriteSize(Uniform Curve Sprite Scale=DataInterface() {propertyValues {Curve {keys [{interpMode=RCIM_Cubic, tangentMode=RCTM_User, time=0, value=1}, {interpMode=RCIM_Cubic, tangentMode=RCTM_User, time=1, value=0}]}}}) {module=ScaleSpriteSize001}
   &ScaleColor(
    Scale Alpha=&FloatFromCurve(FloatCurve=DataInterface() {propertyValues {Curve {keys [{interpMode=RCIM_Cubic, tangentMode=RCTM_User, time=0, value=1}, {interpMode=RCIM_Cubic, tangentMode=RCTM_User, time=1, value=0}]}}})
    ScaleA=true
    ScaleRGB=false
   )
  ]
  rendererClasses [&NiagaraSpriteRendererProperties]
  renderers [{rendererClass=&NiagaraSpriteRendererProperties}]
 }
 SparksA001 {
  emitterUpdateScript [
   &EmitterState()
   &SpawnBurst_Instantaneous(
    Spawn Count=&RandomRangeInt(Minimum=3, Maximum=7)
    Spawn Probability=0.1
    Use Spawn Probability=true
    UseLoopCountLimit=false
   )
   &SpawnRate(SpawnRate=&RandomRangeFloat(Minimum=8, Maximum=20), Use Spawn Probability=false)
  ]
  particleSpawnScript [
   &InitializeParticle(
    Color=<rgba 0 3.8042 9 0.4>
    Lifetime Max=0.6
    Lifetime Min=0.4
    Lifetime Mode=::NewEnumerator1
    Position Offset=<200 0 154>
    Sprite Rotation Mode=::NewEnumerator2
    Sprite Size Mode=::NewEnumerator1
    Sprite UV Mode=::NewEnumerator3
    Uniform Sprite Size Max=3.5
    Uniform Sprite Size Min=2
    UsePositionOffset=true
    Write Material Random=true
   ) {module=InitializeParticle001}
   &ShapeLocation(MaskSpawnGroup=false, Override Seed=false, Sphere Radius=50, UseOwnerScale=true) {module=ShapeLocation001}
   &AddVelocity(Velocity=<50 -55 55>)
  ]
  particleUpdateScript [
   &ParticleState()
   &CurlNoiseForce(Noise Frequency=25, Noise Strength=5, Pan Noise=true, Pan Noise Field=<0 0 0.1>)
   &Drag(Drag=&RandomRangeFloat(Minimum=0.8, Maximum=1.2), Use Linear Drag=true, Use Rotational Drag=false)
   &ScaleSpriteSize(Scale Factor=&Vector2DFromFloat(Value=&FloatFromCurve(FloatCurve=DataInterface() {propertyValues {Curve {keys [
          <at 0 0 RCIM_Cubic>
          <at 0.112 1.0173>
          {arriveTangent=-1.142, interpMode=RCIM_Cubic, leaveTangent=-1.142, time=0.7001, value=0.872}
          <at 1.0011 0.002>
         ]}}})))
   &AccelerationForce(Acceleration=&VectorFromCurve(
     VectorCurve=DataInterface() {propertyValues {
       XCurve {keys [
         {interpMode=RCIM_Cubic, leaveTangent=3.2171, tangentMode=RCTM_Break, time=0, value=0}
         {interpMode=RCIM_Cubic, tangentMode=RCTM_Break, time=1, value=1}
        ]}
       YCurve {keys [<at 0 1>,<at 1 0>]}
       ZCurve {keys [<at 0 1>,<at 1 0>]}
      }}
     Scale Curve=<9999 999 1>
    )) {module=AccelerationForce001}
   &SolveForcesAndVelocity(Clamp Velocity=false, Limit Acceleration=false)
   &ScaleSpriteSizeBySpeed(Max Scale Factor=<1 5>, Velocity Threshold=250)
   &ScaleSpriteSize(Uniform Curve Sprite Scale=DataInterface() {propertyValues {Curve {keys [{interpMode=RCIM_Cubic, tangentMode=RCTM_User, time=0, value=1}, {interpMode=RCIM_Cubic, tangentMode=RCTM_User, time=1, value=0}]}}}) {module=ScaleSpriteSize001}
   &ScaleColor(
    Scale Alpha=&FloatFromCurve(FloatCurve=DataInterface() {propertyValues {Curve {keys [{interpMode=RCIM_Cubic, tangentMode=RCTM_User, time=0, value=1}, {interpMode=RCIM_Cubic, tangentMode=RCTM_User, time=1, value=0}]}}})
    ScaleA=true
    ScaleRGB=false
   )
  ]
  rendererClasses [&NiagaraSpriteRendererProperties]
  renderers [{rendererClass=&NiagaraSpriteRendererProperties}]
 }
}
summary {
 systemName=NS_Launch_LowGravDirect_Burst
 emitters [
  {emitterName=Glow_Idle, rendererClasses [&NiagaraSpriteRendererProperties, &NiagaraLightRendererProperties]}
  {emitterName=Ribbons, rendererClasses [&NiagaraRibbonRendererProperties]}
  {emitterName=Cone_Ray, rendererClasses [&NiagaraMeshRendererProperties]}
  {emitterName=Cone_Sprites, rendererClasses [&NiagaraSpriteRendererProperties]}
  {emitterName=Cone_Rings, rendererClasses [&NiagaraMeshRendererProperties]}
  {emitterName=Cone_Organic, rendererClasses [&NiagaraMeshRendererProperties]}
  {emitterName=SparksA001, rendererClasses [&NiagaraSpriteRendererProperties]}
  {emitterName=Glow_Wispy, rendererClasses [&NiagaraSpriteRendererProperties, &NiagaraLightRendererProperties]}
  {emitterName=SparksA, rendererClasses [&NiagaraSpriteRendererProperties]}
 ]
 userVariables [User.Direction {type=Vector3f, description="", defaultValue {struct=Vector3f, value=<1 0 0>}}]
}
